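# Parent character material for the FifthFully Ring skin, adding a maskable glowing tattoo with Fresnel-scaled emissive.
# PreFX output is blended toward tattoo base and emissive colors using a selectable Tech Art Mask channel multiplied by an enable scalar.
# A UseEyes static switch substitutes the shared eye function for that result.
# The chosen attributes pass through LoneDice's shading function, borrowed from another character, then the shared PostFX function.
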Material /BRCosmetics/Characters/Player/Female/Medium/Bodies/F_MED_FifthFully/Skins/Ring/Materials/M_FifthFully_Shading_Parent
graph {
 properties {
  bAutomaticallySetUsageInEditor=false
  bUseMaterialAttributes=true
  bUsedWithMeshDeformer=true
  bUsedWithMorphTargets=true
  bUsedWithSkeletalMesh=true
  pixelDepthOffsetMode=PDOM_Legacy
  shadingModel=MSM_FromMaterialExpression
  subsurfaceProfile=&SS_HeroSkin_02
  usedShadingModels=Default Lit | Subsurface | Clear Coat | Subsurface Profile | Cloth
 }
 nodes [
  n1 = MaterialFunctionCall() {materialFunction=/Game/Global/Materials/Character_Parent_2023/Eye_Parent_2023/MF_eyes_2023}
  n2 = MaterialFunctionCall() {materialFunction=&MF_CharacterParent_2023_PreFX}
  n3 = VectorParameter() {group=FifthFully, parameterName=Tattoo_Emissive, defaultValue=<rgba 0.15 0 3 1>}
  n4 = LinearInterpolate(
   B=ScalarParameter() {group=FifthFully, parameterName=Tattoo_FresnelEmissiveScale}
   Alpha=Power(Base=@n2.Fresnel, Exp=ScalarParameter() {defaultValue=1, group=FifthFully, parameterName=Tattoo_FresnelExp})
  ) {constA=1}
  n5 = BlendMaterialAttributes(
   A=@n2.Result
   B=SetMaterialAttributes!(
    MaterialAttributes=@n2.Result
    Base Color=VectorParameter() {group=FifthFully, parameterName=Tattoo_BaseColor, defaultValue=<rgba 0 0.0603 1 1>}.RGB
    Emissive Color=Multiply(A=@n3.RGB, B=@n4)
   ) {attributeSetTypes [#1, #2]}
   Alpha=Multiply(
    A=ChannelMaskParameter(None=@n2.Tech Art Mask) {group=FifthFully, parameterName=TattooGlowMaskChannel}.R
    B=ScalarParameter() {group=FifthFully, parameterName=EnableTattooGlow}
   )
  )
  MaterialFunctionCall(
   Material=MaterialFunctionCall!(MA_In=StaticSwitchParameter(True=@n1.Result, False=@n5) {group=LoneDice, parameterName=UseEyes}, M=@n2.M) {materialFunction=/BRCosmetics/Characters/Player/Male/Medium/Bodies/M_MED_LoneDice/Materials/MF_LoneDice_Shading}.MA_Out
   M=@n2.M
   Pre Skinned Local Position=@n2.M
   Distance Blend=@n2.M
   Fresnel=@n2.M
  ) {materialFunction=&MF_CharacterParent_2023_PostFX}
 ]
}
